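# Niagara system for the IgniteEgg_Jab emote, named as its text bubble effect.
# It holds two emitters, OK_Bubble, which is disabled, and OK_Bubble001, which is the active one.
# The only listed module is SystemState, and no renderers, data interfaces or user variables appear, so the bubble's look and motion are not shown here.
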NiagaraSystem /BRCosmetics/Animation/Game/MainPlayer/Emotes/IgniteEgg_Jab/FX/NS_Emote_IgniteEgg_Jab_TextBubble
dependencies {usedDataInterfaces [], usedDynamicInputs [], usedModules [&SystemState]}
emitters {OK_Bubble {bEnabled=false, rendererClasses [], renderers []}, OK_Bubble001 {rendererClasses [], renderers []}}
summary {
 systemName=NS_Emote_IgniteEgg_Jab_TextBubble
 emitters [{bEnabled=false, emitterName=OK_Bubble, rendererClasses []}, {emitterName=OK_Bubble001, rendererClasses []}]
 userVariables []
}
